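# Niagara player warp effect for the Skyfire backbling skin swap, built around the character's User.SkeletalMesh.
# Ring spawns tapering groups of mesh and sprite particles that FN_BoneFlow moves along bones, with time-driven waveforms pulsing their size and blue-tinted color.
# NE_Glow001 adds a one-shot blue sprite pinned near a sampled bone or socket, swelling then fading.
# FN_Debug is disabled, and no shown module reads User.TargePosition.
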NiagaraSystem /Skyfire/Cosmetics/Backbling/Effects/PlayerWarp/NS_Skyfire_PlayerWarp_SkinSwap
dependencies {
 usedDataInterfaces [&NiagaraDataInterfaceCurve]
 usedDynamicInputs []
 usedModules [
  &self:SystemSpawnScript.SystemScriptSource.SystemScriptGraph.NiagaraNodeAssignment_0.SetVariables_#1
  &SystemState
 ]
}
emitters {
 FN_Debug {
  bEnabled=false
  emitterUpdateScript [
   &EmitterState()
   &SpawnBurst_Instantaneous(Spawn Count=1, Use Spawn Probability=false)
  ]
  particleSpawnScript [/Niagara/Modules/Spawn/Initialization/InitializeParticle(
    Lifetime=5
    Module.Write Color=true
    Module.Write Lifetime=true
    Module.Write Mass=true
    Module.Write Material Random=true
    Module.Write Position=false
    Module.Write Scale=false
    Module.Write SpriteRotation=false
    Module.Write SpriteSize=true
    Position=&SimulationPosition()
    Sprite Size=<100 100>
   )]
  particleUpdateScript [
   &ParticleState()
   /SetVariables(Particles.Debug=5555) {isSetParameters=true}
   &DynamicMaterialParameters(
    Index 0 Param 1=$Particles.Debug
    Module.Param0WriteEnabled=true
    Module.Param1WriteEnabled=true
    Module.Param2WriteEnabled=true
    Module.Param3WriteEnabled=true
   )
   &CameraOffset(Camera Offset Amount=50)
  ]
  rendererClasses [&NiagaraSpriteRendererProperties]
  renderers [{rendererClass=&NiagaraSpriteRendererProperties}]
 }
 NE_Glow001 {
  emitterUpdateScript [
   &EmitterState(Life Cycle Mode=::NewEnumerator1, Loop Behavior=::NewEnumerator1, Loop Duration=1, UseLoopDelay=false)
   &SpawnBurst_Instantaneous(Spawn Count=1, Use Spawn Probability=false)
  ]
  particleSpawnScript [
   &InitializeParticle(
    Color=<rgba 0 0.7143 5 1>
    Lifetime=0.6
    Module.Write Color=true
    Module.Write Lifetime=true
    Module.Write Mass=true
    Module.Write Material Random=true
    Module.Write Position=false
    Module.Write Scale=false
    Module.Write SpriteRotation=false
    Module.Write SpriteSize=true
    Position=&SimulationPosition()
    Sprite Size=<180 180>
   )
   &CameraOffset(Camera Offset Amount=50)
  ]
  particleUpdateScript [
   &ParticleState()
   &SkeletalMeshLocation(
    Bone / Socket Index=13
    Bone Sampling Mode=::NewEnumerator4
    Inherit Velocity=false
    MaskSpawnGroup=false
    Sampled Position Offset=<0 0 20>
    Skeletal Mesh=$User.SkeletalMesh
   )
   &ScaleSpriteSize(Scale Factor=&Vector2DFromFloat(Value=&FloatFromCurve(FloatCurve=DataInterface() {propertyValues {Curve {keys [<at 0 0>,<at 0.2 1>,<at 1 0.75>]}}}, Scale Curve=2)))
   &ScaleColor(
    Scale Alpha=&FloatFromCurve(FloatCurve=DataInterface() {propertyValues {Curve {keys [<at 0 1 RCIM_Cubic>, {arriveTangent=-2.9824, interpMode=RCIM_Cubic, leaveTangent=-2.9824, tangentMode=RCTM_User, time=1, value=0}]}}})
    ScaleA=true
    ScaleRGB=true
   )
  ]
  rendererClasses [&NiagaraSpriteRendererProperties]
  renderers [{rendererClass=&NiagaraSpriteRendererProperties}]
 }
 Ring {
  emitterSpawnScript [/SetVariables(Emitter.UseParticleColorInput=1) {isSetParameters=true}]
  emitterUpdateScript [
   &EmitterState()
   &SpawnRate(
    SpawnRate=&Lerp_Float(
     A=0
     B=70
     Alpha=&FloatFromCurve(FloatCurve=DataInterface() {propertyValues {Curve {keys [<at 0 1>,<at 1 0>]}}}, CurveIndex=$Emitter.NormalizedLoopAge)
    )
    Use Spawn Probability=false
   )
  ]
  particleSpawnScript [
   &FN_SetSpawnGroup(Inclusivity=true, Normalize Spawn Group=true, Particles Per Group=6, Pattern=::NewEnumerator0, Set Ribbon ID and Ribbon Link Order=true) {module=FN_SetSpawnGroup001}
   /Niagara/Modules/Spawn/Initialization/V2/InitializeParticle(
    Lifetime Max=0.4
    Lifetime Min=0.2
    Lifetime Mode=::NewEnumerator1
    Mesh Scale Mode=::NewEnumerator3
    Mesh Uniform Scale=0.07
    UsePositionOffset=false
    Write Material Random=true
   )
   &CameraOffset(Camera Offset Amount=8)
  ]
  particleUpdateScript [
   &ParticleState()
   /Game/Effects/Niagara/Modules/BoneFlow/FN_BoneFlow_v4/FN_BoneFlow(
    Angle:NiagaraFloat=&Multiply_Float(A=$Particles.Normalized Spawn Group, B=$Engine.Time)
    BoneID:NiagaraInt32=&ModuloInt(A=$Particles.SpawnGroup, B=5)
    Height:NiagaraFloat=&Multiply_Float(A=$Particles.Normalized Spawn Group, B=0.9)
    Interpolate Mesh Sampling=false
    Radius:NiagaraFloat=&FloatFromCurve(
     FloatCurve=DataInterface() {propertyValues {Curve {keys [<at -0.0028 0.3479 RCIM_Cubic>, {arriveTangent=-0.0477, interpMode=RCIM_Cubic, leaveTangent=-0.0477, time=0.4906, value=1}, <at 1 0.3>]}}}
     CurveIndex=$Particles.Normalized Spawn Group
     Scale Curve=2
    )
    SideVector1:Vector3f=<5 5 5>
    SideVector2:Vector3f=<0 0 0>
    SideVector3:Vector3f=<1 1 1>
    SkeletalMesh:NiagaraDataInterfaceSkeletalMesh=$User.SkeletalMesh
    Stabalize Twisting=true
    Write Position Attribute=true
   ) {module=FN_BoneFlow001}
   /SetVariables(Particles.SizeWave:NiagaraFloat=&Waveform(
     Global Input Position=&Multiply_Float(A=$Engine.Time, B=3)
     Global Phase=$Particles.NormalizedAge
     "Waveform [1]"=::NewEnumerator0
     "[1] Phase"=$Particles.NormalizedAge
     "[1] Frequency"=0.3
     "[1] Amplitude Min/Max"=<1 0>
     "[1] Scale"=1.5
     "[1] Bias"=1
     "[2] Phase"=$Emitter.Age
     "[3] Phase"=$Emitter.Age
    )) {isSetParameters=true}
   /SetVariables(Particles.Desaturation=0.6) {isSetParameters=true}
   /SetVariables(Particles.WaveColor:LinearColor=&MakeLinearColor(
     R:NiagaraFloat=&Waveform(
      Global Input Position=&Multiply_Float(A=$Engine.Time, B=1.5)
      Global Phase=$Particles.NormalizedAge
      "Waveform [1]"=::NewEnumerator0
      "[1] Phase"=$Particles.NormalizedAge
      "[1] Frequency"=1.5
      "[1] Amplitude Min/Max"=<1 0>
      "[1] Bias"=$Particles.Desaturation
      "[2] Phase"=$Emitter.Age
      "[3] Phase"=$Emitter.Age
     )
     G:NiagaraFloat=&Waveform(
      Global Input Position=&Multiply_Float(A=$Engine.Time, B=1)
      Global Phase=$Particles.NormalizedAge
      "Waveform [1]"=::NewEnumerator1
      "[1] Phase"=$Particles.NormalizedAge
      "[1] Frequency"=1
      "[1] Amplitude Min/Max"=<1 0>
      "[1] Scale"=2
      "[1] Bias"=$Particles.Desaturation
      "[2] Phase"=$Emitter.Age
      "[3] Phase"=$Emitter.Age
     )
     B:NiagaraFloat=&Waveform(
      Global Input Position=&Multiply_Float(A=$Engine.Time, B=1.5)
      Global Phase=$Particles.NormalizedAge
      "Waveform [1]"=::NewEnumerator2
      "[1] Phase Scale"=2.5
      "[1] Phase"=$Particles.NormalizedAge
      "[1] Frequency"=2
      "[1] Amplitude Min/Max"=<1 0>
      "[1] Bias"=$Particles.Desaturation
      "[2] Phase"=$Emitter.Age
      "[3] Phase"=$Emitter.Age
     )
     A:NiagaraFloat=1
    )) {isSetParameters=true}
   /SetVariables(Particles.Color=&MultiplyLinearColor(
     Linear Color=$Particles.WaveColor
     Scale Factor=&MakeVector4_FromLinearColor(LinearColor=&MakeLinearColorFromVectorAndFloat(
       "Vector (RGB)"=<0.2 0.5 1>
       "Float (Alpha)"=&FloatFromCurve(
        FloatCurve=DataInterface() {propertyValues {Curve {keys [
            {arriveTangent=9.0725, interpMode=RCIM_Cubic, leaveTangent=9.0725, tangentMode=RCTM_User, time=0, value=1}
            {arriveTangent=-2.5249, interpMode=RCIM_Cubic, leaveTangent=-2.5249, tangentMode=RCTM_User, time=0.3, value=0.95}
            {arriveTangent=-1.3571, interpMode=RCIM_Cubic, leaveTangent=-1.3571, time=0.6, value=0.2}
            {arriveTangent=-0.0405, interpMode=RCIM_Cubic, leaveTangent=-0.0405, tangentMode=RCTM_User, time=1, value=0}
           ]}}}
        Scale Curve=&FloatFromCurve(
         FloatCurve=DataInterface() {propertyValues {Curve {keys [{interpMode=RCIM_Cubic, tangentMode=RCTM_User, time=0, value=1}, {interpMode=RCIM_Cubic, tangentMode=RCTM_User, time=1, value=0}]}}}
         CurveIndex=$Emitter.NormalizedLoopAge
        )
       )
      ))
    )) {isSetParameters=true}
   /SetVariables(Particles.ParticleDecalColor:LinearColor=&MultiplyLinearColor(
     Linear Color=$Particles.WaveColor
     Scale Factor=&MakeVector4_FromLinearColor(LinearColor=&MakeLinearColorFromVectorAndFloat(
       "Vector (RGB)"=<0.1 0.2 1>
       "Float (Alpha)"=&FloatFromCurve(
        FloatCurve=DataInterface() {propertyValues {Curve {keys [
            {arriveTangent=9.0725, interpMode=RCIM_Cubic, leaveTangent=9.0725, tangentMode=RCTM_User, time=0, value=1}
            {arriveTangent=-2.5249, interpMode=RCIM_Cubic, leaveTangent=-2.5249, tangentMode=RCTM_User, time=0.3, value=0.95}
            {arriveTangent=-1.3571, interpMode=RCIM_Cubic, leaveTangent=-1.3571, time=0.6, value=0.2}
            {arriveTangent=-0.0405, interpMode=RCIM_Cubic, leaveTangent=-0.0405, tangentMode=RCTM_User, time=1, value=0}
           ]}}}
        Scale Curve=&FloatFromCurve(
         FloatCurve=DataInterface() {propertyValues {Curve {keys [{interpMode=RCIM_Cubic, tangentMode=RCTM_User, time=0, value=1}, {interpMode=RCIM_Cubic, tangentMode=RCTM_User, time=1, value=0}]}}}
         CurveIndex=$Emitter.NormalizedLoopAge
         Scale Curve=0.4
        )
       )
      ))
    )) {isSetParameters=true}
   /SetVariables(Particles.BoneFlowTangent:Vector3f=$Output.FN_BoneFlow001.SplineTangent) {isSetParameters=true}
   /SetVariables(Particles.RingRotation:Quat4f=&MakeQuatFromAxisAngle(Axis=$Particles.BoneFlowTangent, AngleInDegrees=120, NormalizeAxis=true)) {isSetParameters=true}
   &ScaleMeshSize(Scale Factor=&Multiply_Vector(
     A=&MakeVector(X=$Particles.SizeWave, Y=$Particles.SizeWave, Z=1)
     B=&VectorFromFloat(Value=&FloatFromCurve(FloatCurve=DataInterface() {propertyValues {Curve {keys [
           {arriveTangent=4.4833, interpMode=RCIM_Cubic, leaveTangent=4.4833, tangentMode=RCTM_User, time=0, value=0.25}
           {arriveTangent=-0.7444, interpMode=RCIM_Cubic, leaveTangent=-0.7444, tangentMode=RCTM_User, time=0.5, value=0.85}
           {arriveTangent=-0.0405, interpMode=RCIM_Cubic, leaveTangent=-0.0405, tangentMode=RCTM_User, time=1, value=0.2758}
          ]}}}))
    )) {module=ScaleMeshSize001}
   &DynamicMaterialParameters(
    Index 0 Param 1=&FloatFromCurve(FloatCurve=DataInterface() {propertyValues {Curve {keys [
         {arriveTangent=9.0725, interpMode=RCIM_Cubic, leaveTangent=9.0725, tangentMode=RCTM_User, time=0, value=0}
         {arriveTangent=0.1068, interpMode=RCIM_Cubic, leaveTangent=0.1068, tangentMode=RCTM_User, time=0.1515, value=0.9243}
         {arriveTangent=-0.0405, interpMode=RCIM_Cubic, leaveTangent=-0.0405, tangentMode=RCTM_User, time=1, value=0}
        ]}}})
    Module.Param0WriteEnabled=true
    Module.Param1WriteEnabled=true
    Module.Param2WriteEnabled=true
    Module.Param3WriteEnabled=true
   )
  ]
  rendererClasses [&NiagaraMeshRendererProperties, &NiagaraSpriteRendererProperties]
  renderers [
   {rendererClass=&NiagaraMeshRendererProperties, rendererIndex=0}
   {rendererClass=&NiagaraSpriteRendererProperties, rendererIndex=1}
  ]
 }
}
summary {
 systemName=NS_Skyfire_PlayerWarp_SkinSwap
 emitters [
  {emitterName=Ring, rendererClasses [&NiagaraMeshRendererProperties, &NiagaraSpriteRendererProperties]}
  {emitterName=NE_Glow001, rendererClasses [&NiagaraSpriteRendererProperties]}
  {bEnabled=false, emitterName=FN_Debug, rendererClasses [&NiagaraSpriteRendererProperties]}
 ]
 userVariables [
  User.SkeletalMesh {
   type=NiagaraDataInterfaceSkeletalMesh
   defaultValue {
    struct=NiagaraExt_VariableValue_DataInterface
    value {dataInterface=&self:NiagaraDataInterfaceSkeletalMesh_0}
   }
  }
  User.TargePosition {type=Vector3f, description="", defaultValue {struct=Vector3f, value=<0 0 0>}}
 ]
}
